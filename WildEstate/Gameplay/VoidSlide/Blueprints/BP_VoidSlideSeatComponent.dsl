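; Actor component that manages player seats on a Void Slide, tracking occupants by replicated net IDs, one slot per seat socket.
; Server functions occupy, vacate or cycle seats; occupy and vacate fire Enter, Exit, Full and Empty events.
; Clients rebuild those events, plus seat changes, by diffing replicated IDs against a cached copy.
; Taking or changing seats grants a knockback ability to actors overlapping a capsule at the seat socket, excluding seated players.
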(event EventBeginPlay
  ;[0 1 0 0.5] [BeginPlay]
  (Utilities|FlowControl|Sequence
    (:then_0
      (bind _asbga_void_slide_parent (Utilities|Casting|CastToBGA_VoidSlide_Parent (Components|GetOwner)))
      (Class|BGAVoidSlideParent|SetUpSeatEventBindings _asbga_void_slide_parent))
    (:then_1
      (Utilities|SwitchHasAuthority
        (:Authority
          ;;[1 1 0 0.5] [SERVER] Size PlayerStates to the number of sockets
          (Utilities|Array|Resize (Variables|Replicated|GetPlayerNetIds) (Utilities|Array|Length (Variables|Assets|GetSeatSocketNames))))
        (:Remote
          ;;[1 0.5 0 0.5] [CLIENT] Broadcast events based on initial replicated value of PlayerStates, in case OnRep_PlayerStates doesn't trigger
          (CallFunction|BroadcastClientEvents))))))

(fn ApplySeatedKnockback
  (params (SeatIndex :type Integer))
  (bind _returnvalue (Components|GetOwner))
  (bind _returnvalue_2 (Components|GetOwner))
  (bind _seatsocketname (Default|GetSeatSocketName SeatIndex))
  (bind _returnvalue_3 (Actor|GetComponentbyClass _returnvalue "/Script/Engine.StaticMeshComponent"))
  (bind _returnvalue_4 (Utilities|GetDisplayName _returnvalue_2))
  (if (Utilities|Array|IsValidIndex (Variables|Assets|GetSeatSocketNames) SeatIndex)
    (if (Utilities|IsValid _returnvalue_3 :!class /Script/Engine.KismetSystemLibrary)
      (if (!= _seatsocketname "")
        (bind _output_get (LocalVariables|SetSeat%20Socket%20Location (Transformation|GetSocketLocation _returnvalue_3 _seatsocketname)))
        (if (Variables|Const|GetbDebug)
          (Rendering|Debug|DrawDebugCapsule _output_get (Variables|Config|GetSeatKnockbackHalfHeight) (Variables|Config|GetSeatKnockbackRadius) "0, 0, 0" "(R=0.000000,G=1.000000,B=0.000000,A=1.000000)" 5.0 1.0)
          (else
            _))
        (bind _outactors (Collision|CapsuleOverlapActors (LocalVariables|GetSeat%20Socket%20Location) (Variables|Config|GetSeatKnockbackRadius) (Variables|Config|GetSeatKnockbackHalfHeight) (Utilities|Array|MakeArray "ObjectTypeQuery3") 0 (Default|GetPlayerStates)))
        (for _array_element _outactors
          (if (Utilities|IsValid (Ability|GetAbilitySystemComponent _array_element) :!class /Script/Engine.KismetSystemLibrary)
            (Ability|Advanced|GrantandActivateAbilityonSelf (Ability|GetAbilitySystemComponent _array_element) (Variables|Assets|GetSeatKnockback) (Components|GetOwner))
            (else
              (Fortnite|FortShippingLog self (Utilities|String|Append _returnvalue_4 " [ApplySeatedKnockback] Player AbilitySystemComponent invalid") true))))
        (else
          (Fortnite|FortShippingLog self (Utilities|String|Append _returnvalue_4 (Utilities|String|Append " [ApplySeatedKnockback] Invalid SeatSocketName index: " (Utilities|String|ToString(Integer) (Variables|GetSeatIndex)))))))
      (else
        (Fortnite|FortShippingLog self (Utilities|String|Append _returnvalue_4 " [ApplySeatedKnockback] VoidSlide Static Mesh Component not found") true)))
    (else
      (Fortnite|FortShippingLog self (Utilities|String|Append _returnvalue_4 (Utilities|String|Append " [ApplySeatedKnockback] Invalid seat index: " (Utilities|String|ToString(Integer) (Variables|GetSeatIndex)))) true))))

(fn BroadcastClientEvents
  (bind _oldplayernetids (Variables|Dynamic|GetOldPlayerNetIds))
  (Utilities|SwitchHasAuthority
    (:Remote
      (Utilities|FlowControl|Sequence
        (:then_0
          ;[0 1 0 0.5 16] Resize OldPlayerNetIds to match PlayerNetIds.
          ; The sizes will initially be zero, until we reach this point.
          ; The size will never decrease, because we never reduce the size of PlayerNetIds.
          (Utilities|Array|Resize (Variables|Dynamic|GetOldPlayerNetIds) (Utilities|Array|Length (Variables|Replicated|GetPlayerNetIds))))
        (:then_1
          ;[1 1 0 0.5] Count how many PlayerNetIds were previously valid.
          (for _array_element (Variables|Dynamic|GetOldPlayerNetIds)
            (if (Online|UniqueNetIdRepl|IsValid(UniqueIdRepl) _array_element)
              (Math|Integer|IncrementInt (LocalVariables|GetNumValidOldPlayerStates)))))
        (:then_2
          ;[1 0.5 0 0.5] Call OnChangedSeat for all players who changed seats since the last update.
          ; Call OnExit for all players who exited since the last update.
          (for (_array_element_2 _array_index) (Variables|Dynamic|GetOldPlayerNetIds)
            (if (Online|UniqueNetIdRepl|IsValid(UniqueIdRepl) _array_element_2)
              (bind _output_get (LocalVariables|SetNetId _array_element_2))
              (bind _asfort_player_state_athena (Utilities|Casting|CastToFortPlayerStateAthena (Game|GetPlayerStatefromUniqueNetId _output_get))
                (:then
                  _)
                (:CastFailed
                  _)) ; Possibly went null, due to leaving match
              (LocalVariables|SetOldPS _asfort_player_state_athena)
              (LocalVariables|SetOldSeatIndex _array_index)
              (bind _output_get_2 (LocalVariables|SetNewSeatIndex (Utilities|Array|FindItem (Variables|Replicated|GetPlayerNetIds) (LocalVariables|GetNetId))))
              (if (>= _output_get_2 0)
                (if (!= (LocalVariables|GetNewSeatIndex) (LocalVariables|GetOldSeatIndex))
                  (Fortnite|FortShippingLog self (Utilities|String|Append "[BroadcastEvents] Player [" (Fortnite|GetPlayerNameSafe (LocalVariables|GetOldPS)) "] changed from seat [" (Utilities|String|ToString(Integer) (LocalVariables|GetOldSeatIndex)) "] to [" (Utilities|String|ToString(Integer) (LocalVariables|GetNewSeatIndex)) "]"))
                  (Default|CallOnChangedSeat (LocalVariables|GetOldSeatIndex) (LocalVariables|GetNewSeatIndex) (LocalVariables|GetNetId) (LocalVariables|GetOldPS)))
                (else
                  (Fortnite|FortShippingLog self (Utilities|String|Append "[BroadcastEvents] Player [" (Fortnite|GetPlayerNameSafe (LocalVariables|GetOldPS)) "] exited seat [" (Utilities|String|ToString(Integer) (LocalVariables|GetOldSeatIndex)) "]"))
                  (Default|CallOnExit (LocalVariables|GetOldSeatIndex) (LocalVariables|GetNetId) (LocalVariables|GetOldPS)))))))
        (:then_3
          ;[0.1489 0.04509 0.15 0.5] Call OnEnter for all players who entered since the last update
          (for (_array_element_3 _array_index_2) (Variables|Replicated|GetPlayerNetIds)
            (bind _output_get_3 (LocalVariables|SetNetId _array_element_3))
            (if (Online|UniqueNetIdRepl|IsValid(UniqueIdRepl) _output_get_3)
              (bind _asfort_player_state_athena_2 (Utilities|Casting|CastToFortPlayerStateAthena (Game|GetPlayerStatefromUniqueNetId _output_get_3))
                (:then
                  _)
                (:CastFailed
                  _)) ; Should we keep going when we have a valid net id,\nbut the PS is somehow null?
              (LocalVariables|SetNewPS _asfort_player_state_athena_2)
              (if (not (Utilities|Array|ContainsItem (Variables|Dynamic|GetOldPlayerNetIds) (LocalVariables|GetNetId)))
                (LocalVariables|SetNewSeatIndex _array_index_2)
                (Fortnite|FortShippingLog self (Utilities|String|Append "[BroadcastEvents] Player [" (Fortnite|GetPlayerNameSafe (LocalVariables|GetNewPS)) "] entered at seat [" (Utilities|String|ToString(Integer) (LocalVariables|GetOldSeatIndex)) "]"))
                (Default|CallOnEnter (LocalVariables|GetOldSeatIndex) (LocalVariables|GetNetId) (LocalVariables|GetNewPS))))))
        (:then_4
          ;[0 0 1 0.5] Check if PlayerStates became full or empty
          (if (Default|IsFullandHasatLeastOneSeat)
            (if (< (LocalVariables|GetNumValidOldPlayerStates) (Default|GetNumSeats))
              (Fortnite|FortShippingLog self "[BroadcastEvents] Became full")
              (Default|CallOnFull))
            (elif (Default|HasAnyOccupant)
              _
              (elif (> (LocalVariables|GetNumValidOldPlayerStates) 0)
                (Fortnite|FortShippingLog self "[BroadcastEvents] Became empty")
                (Default|CallOnEmpty)))))
        (:then_5
          ;[0 0.15 0.15 0.5] Now that we have finished comparing old and new data, sync the old data to the new
          (Utilities|Array|Clear _oldplayernetids)
          (Utilities|Array|AppendArray _oldplayernetids (Variables|Replicated|GetPlayerNetIds)))
        (:then_6
          ;[0.1397 0 0.15 0.5] Log old/new occupant count
          (Fortnite|FortShippingLog self (Utilities|String|Append "[BroadcastEvents] OldNumOccupants [" (Utilities|String|ToString(Integer) (LocalVariables|GetNumValidOldPlayerStates)) "] NewNumOccupants [" (Utilities|String|ToString(Integer) (Default|GetNumOccupants)) "]")))))))

(fn ContainsPlayerState
  (params (PlayerState :type Fort%20Player%20State%20Object%20Reference%20%28by%20ref%29))
  (bind _returnvalue (PlayerState|GetPlayerUniqueNetID PlayerState))
  (Utilities|FlowControl|Sequence
    (:then_0
      (Utilities|IsValid PlayerState
        (:Is%20Valid
          (if (Online|UniqueNetIdRepl|IsValid(UniqueIdRepl) _returnvalue)
            (return (Utilities|Array|ContainsItem (Variables|Replicated|GetPlayerNetIds) _returnvalue))))))
    (:then_1
      (return false))))

(fn GetFirstAvailableSeatIndex
  (bind _playernetids (Variables|Replicated|GetPlayerNetIds))
  (bind _returnvalue (Utilities|Array|FindItem _playernetids))
  (bind _returnvalue_2 (Utilities|Array|Length _playernetids))
  (if (>= _returnvalue 0)
    (return _returnvalue)
    (elif (< _returnvalue_2 (Default|GetNumSeats))
      (return _returnvalue_2)
      (else
        (return -1)))))

(fn GetNextAvailableSeatIndex
  (params (OldSeatIndex :type Integer))
  (Utilities|FlowControl|Sequence
    (:then_0
      (LocalVariables|SetNumSeats (Default|GetNumSeats))
      (if (Math|Integer|InRange(Integer) OldSeatIndex 0 (- (LocalVariables|GetNumSeats) 1))
        (LocalVariables|SetNextSeatIndex (Math|Integer|%25%28Integer%29 (+ (Variables|GetOldSeatIndex) 1) (LocalVariables|GetNumSeats)))
        (while (!= (LocalVariables|GetNextSeatIndex) (Variables|GetOldSeatIndex))
          (if (Online|UniqueNetIdRepl|IsValid(UniqueIdRepl) (Utilities|Array|Get(acopy) (Variables|Replicated|GetPlayerNetIds) (LocalVariables|GetNextSeatIndex)))
            (LocalVariables|SetNextSeatIndex (Math|Integer|%25%28Integer%29 (+ (LocalVariables|GetNextSeatIndex) 1) (LocalVariables|GetNumSeats)))
            (else
              (return (LocalVariables|GetNextSeatIndex)))))))
    (:then_1
      (return -1))))

(fn GetNumOccupants
  (for _array_element (Variables|Replicated|GetPlayerNetIds)
    (if (Online|UniqueNetIdRepl|IsValid(UniqueIdRepl) _array_element)
      (Math|Integer|IncrementInt (LocalVariables|GetNumOccupants))))
  (return (LocalVariables|GetNumOccupants)))

(fn GetNumSeats
  (return (Utilities|Array|Length (Variables|Replicated|GetPlayerNetIds))))

(fn GetPlayerState
  (params (SeatIndex :type Integer))
  (bind _playernetids (Variables|Replicated|GetPlayerNetIds))
  (bind _output (Utilities|Array|Get(acopy) _playernetids SeatIndex))
  (Utilities|FlowControl|Sequence
    (:then_0
      (if (Utilities|Array|IsValidIndex (Variables|Replicated|GetPlayerNetIds) SeatIndex)
        (if (Online|UniqueNetIdRepl|IsValid(UniqueIdRepl) _output)
          (bind _asfort_player_state_athena (Utilities|Casting|CastToFortPlayerStateAthena (Game|GetPlayerStatefromUniqueNetId _output)))
          (return _asfort_player_state_athena))))
    (:then_1
      (return 0))))

(fn GetPlayerStates
  (bind _playernetids (Variables|Replicated|GetPlayerNetIds))
  (Utilities|Array|Resize (LocalVariables|GetPSArray) (Utilities|Array|Length _playernetids))
  (for (_array_element _array_index) _playernetids
    (bind _asfort_player_state_athena (Utilities|Casting|CastToFortPlayerStateAthena (Game|GetPlayerStatefromUniqueNetId _array_element)))
    (Utilities|Array|SetArrayElem (LocalVariables|GetPSArray) _array_index _asfort_player_state_athena))
  (return (LocalVariables|GetPSArray)))

(fn GetSeatIndex
  (params (PlayerState :type Fort%20Player%20State%20Athena%20Object%20Reference))
  (Utilities|IsValid PlayerState
    (:Is%20Valid
      (return (Utilities|Array|FindItem (Variables|Replicated|GetPlayerNetIds) (PlayerState|GetPlayerUniqueNetID PlayerState))))
    (:Is%20Not%20Valid
      (return -1))))

(fn GetSeatSocketName
  (params (SeatIndex :type Integer))
  (bind _seatsocketnames (Variables|Assets|GetSeatSocketNames))
  (if (and (< SeatIndex (Default|GetNumSeats)) (Utilities|Array|IsValidIndex _seatsocketnames SeatIndex))
    (return (Utilities|Array|Get(acopy) _seatsocketnames SeatIndex))
    (else
      (Fortnite|FortShippingLog self (Utilities|String|Append (Utilities|String|Append (Utilities|GetDisplayName (Components|GetOwner)) " [GetSeatSocketName] Invalid Seat Index: ") (Utilities|String|ToString(Integer) SeatIndex)) true)
      (return "None"))))

(fn HasAnyOccupant
  (for _array_element (Variables|Replicated|GetPlayerNetIds)
    (if (Online|UniqueNetIdRepl|IsValid(UniqueIdRepl) _array_element)
      (return true)))
  (return false))

(fn HasAvailableSeat
  (return (>= (CallFunction|GetFirstAvailableSeatIndex) 0)))

(fn HasLocallyControlledPlayers
  (for _array_element (Variables|Replicated|GetPlayerNetIds)
    (bind _asfort_player_state_athena (Utilities|Casting|CastToFortPlayerStateAthena (Game|GetPlayerStatefromUniqueNetId _array_element)))
    (if (Class|FortPlayerState|HasLocalPlayerController _asfort_player_state_athena)
      (return true)))
  (return false))

(fn IsFullAndHasAtLeastOneSeat
  (return (and (> (Default|GetNumSeats) 0) (>= (Default|GetNumOccupants) (Default|GetNumSeats)))))

(fn OnRep_PlayerNetIds
  (Utilities|SwitchHasAuthority
    (:Remote
      (CallFunction|BroadcastClientEvents) ; Broadcast events on CLIENT
    )) ; We already broadcast the events on server, in TryOccupy and Unoccupy
)

(fn TryChangeSeat
  (params (PlayerState :type Fort%20Player%20State%20Athena%20Object%20Reference))
  (bind _newseatindex (LocalVariables|GetNewSeatIndex))
  (bind _playerstate (Variables|GetPlayerState))
  (Utilities|FlowControl|Sequence
    (:then_0
      (Utilities|IsValid _playerstate
        (:Is%20Valid
          (bind _output_get (LocalVariables|SetOldSeatIndex (Default|GetSeatIndex _playerstate)))
          (if (>= _output_get 0)
            (bind _output_get_2 (LocalVariables|SetNewSeatIndex (Default|GetNextAvailableSeatIndex (LocalVariables|GetOldSeatIndex))))
            (if (>= _output_get_2 0)
              (Utilities|Array|SetArrayElem (Variables|Replicated|GetPlayerNetIds) (LocalVariables|GetOldSeatIndex)) ; Clear old slot
              (Utilities|Array|SetArrayElem (Variables|Replicated|GetPlayerNetIds) (LocalVariables|GetNewSeatIndex) (PlayerState|GetPlayerUniqueNetID (Variables|GetPlayerState))) ; Fill new slot
              (Fortnite|FortShippingLog self (Utilities|String|Append "[TryChangeSeat] Moving player [" (Fortnite|GetPlayerNameSafe (Variables|GetPlayerState)) "] from seat [" (Utilities|String|ToString(Integer) (LocalVariables|GetOldSeatIndex)) "] to seat [" (Utilities|String|ToString(Integer) (LocalVariables|GetNewSeatIndex)) "]"))
              (CallFunction|ApplySeatedKnockback _newseatindex)
              (return true _newseatindex))))))
    (:then_1
      (return false (LocalVariables|GetOldSeatIndex)))))

(fn TryOccupy
  (params (PlayerState :type Fort%20Player%20State%20Athena%20Object%20Reference))
  (Utilities|FlowControl|Sequence
    (:then_0
      (Utilities|SwitchHasAuthority
        (:Authority
          (Utilities|IsValid PlayerState
            (:Is%20Valid
              (if (not (Default|ContainsPlayerState PlayerState))
                (bind _output_get (LocalVariables|SetSeatIndex (CallFunction|GetFirstAvailableSeatIndex)))
                (if (>= _output_get 0)
                  (Fortnite|FortShippingLog self (Utilities|String|Append "[TryOccupy] Placing player [" (Fortnite|GetPlayerNameSafe PlayerState) "] in Seat [" (Utilities|String|ToString(Integer) (LocalVariables|GetSeatIndex)) "]"))
                  (Utilities|Array|SetArrayElem (Variables|Replicated|GetPlayerNetIds) (LocalVariables|GetSeatIndex) (PlayerState|GetPlayerUniqueNetID PlayerState)) ; Track the player in the desired seat index of the PlayerStates array, and replicate
                  (Default|CallOnEnter :SeatIndex (LocalVariables|GetSeatIndex) :PlayerState PlayerState)
                  (CallFunction|ApplySeatedKnockback (LocalVariables|GetSeatIndex))
                  (if (Default|IsFullandHasatLeastOneSeat)
                    (Fortnite|FortShippingLog self "[TryOccupy] Became full")
                    (Default|CallOnFull)
                    (else
                      _))
                  (return (LocalVariables|GetSeatIndex))
                  (else
                    ;[1 0 0 0.5] FAILURE CASES
                    (Fortnite|FortShippingLog self (Utilities|String|Append "[TryOccupy] Player [" (Fortnite|GetPlayerNameSafe (Variables|GetPlayerState)) "] failed due to GetAvailableSeatIndex returning -1") true)))
                (else
                  (Fortnite|FortShippingLog self (Utilities|String|Append "[TryOccupy] Player [" (Fortnite|GetPlayerNameSafe (Variables|GetPlayerState)) "] already in Seat [" (Utilities|String|ToString(Integer) (LocalVariables|GetSeatIndex)) "]") true))))
            (:Is%20Not%20Valid
              (Fortnite|FortShippingLog self "[TryOccupy] PlayerState is null" true))))
        (:Remote
          (Fortnite|FortShippingLog self "[TryOccupy] Not Authority" true))))
    (:then_1
      (return -1))))

(fn Unoccupy
  (params (PlayerState :type Fort%20Player%20State%20Athena%20Object%20Reference))
  (Utilities|SwitchHasAuthority
    (:Authority
      (Utilities|IsValid PlayerState
        (:Is%20Valid
          (bind _output_get (LocalVariables|SetSeatIndex (Utilities|Array|FindItem (Variables|Replicated|GetPlayerNetIds) (PlayerState|GetPlayerUniqueNetID PlayerState))))
          (if (>= _output_get 0)
            (Fortnite|FortShippingLog self (Utilities|String|Append "[Unoccupy] Removing player [" (Fortnite|GetPlayerNameSafe PlayerState) "] from Seat [" (Utilities|String|ToString(Integer) (LocalVariables|GetSeatIndex)) "]"))
            (Utilities|Array|SetArrayElem (Variables|Replicated|GetPlayerNetIds) _output_get) ; Clear the seat index of the PlayerNetIds array, and replicate
            (Default|CallOnExit _output_get (PlayerState|GetPlayerUniqueNetID (Variables|GetPlayerState)) (Variables|GetPlayerState))
            (if (not (Default|HasAnyOccupant))
              (Fortnite|FortShippingLog self "[TryOccupy] Became empty")
              (Default|CallOnEmpty))
            (else
              (Fortnite|FortShippingLog self (Utilities|String|Append "[Unoccupy] Player [" (Fortnite|GetPlayerNameSafe PlayerState) "] is NOT seated") true))))
        (:Is%20Not%20Valid
          (Fortnite|FortShippingLog self "[Unoccupy] PlayerState is null" true))))
    (:Remote
      (Fortnite|FortShippingLog self "[Unoccupy] Not Authority" true))))
