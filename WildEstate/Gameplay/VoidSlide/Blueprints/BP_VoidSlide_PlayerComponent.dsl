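; Player pawn component that manages sitting on a VoidSlide couch, from attachment to a launched exit.
; BeginPlay links animation, holsters weapons, blocks abilities via data-table tags, applies server effects and sets up local input and HUD.
; Initialize attaches the pawn to its seat socket and overrides the camera, while clients re-poll to correct replicated transforms.
; Timer-gated exit or damage triggers StandUp, which reverses setup, ejects the pawn and later destroys the component.
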(event Custom|OnCanStandTimerEnded)

(event EventBeginPlay
  (bind _pawn (Variables|CachedPointers|GetPawn))
  ;[0 1 0 0.5] [BeginPlay]
  (bind _asfort_player_pawn (Utilities|Casting|CastToFortPlayerPawn (Components|GetOwner))
    (:then
      _)
    (:CastFailed
      _)) ; Assume this can never fail
  (bind _output_get (Variables|CachedPointers|SetPawn _asfort_player_pawn)) ; Cache owning pawn
  (Animation|LinkAnimLayer _output_get "/WildEstate/Gameplay/VoidSlide/Sequence/VoidSlide_Player_AnimBP.VoidSlide_Player_AnimBP_C")
  (bind _asfort_player_state_athena (Utilities|Casting|CastToFortPlayerStateAthena (Variables|/Script/Engine.Pawn|GetPlayerState (Variables|CachedPointers|GetPawn)))
    (:then
      _)
    (:CastFailed
      _)) ; Assume this can never fail
  (bind _output_get_2 (Variables|CachedPointers|SetPlayerState _asfort_player_state_athena)) ; Cache FortPlayerStateAthena
  (bind _asfort_player_controller_athena (Utilities|Casting|CastToFortPlayerControllerAthena (PlayerState|GetPlayerController _output_get_2))
    (:then
      _)
    (:CastFailed
      _)) ; Will be null for non-locally controlled pawns
  (Variables|CachedPointers|SetController _asfort_player_controller_athena) ; Cache FortPlayerController
  (Variables|CachedPointers|SetASC (Ability|GetFortAbilitySystemComponent _output_get_2)) ; Cache AbilitySystemComponent
  (Weapon|SetHolsterWeapon (Variables|CachedPointers|GetPawn) 0 true true (Variables|Const|GetHolsterId))
  (Default|CancelAbilities)
  (Default|BlockAbilities)
  (Utilities|SwitchHasAuthority
    (:Authority
      ;;[1 1 1] Server-only setup
      (Variables|Dynamic|SetPreviousHealth (FortPawn|GetHealth _pawn))
      (Variables|Dynamic|SetPreviousShield (FortPawn|GetShield _pawn))
      (Health|BindEventtoOnPawnHealthChanged _pawn (EventDispatchers|CreateEvent :!event Custom|HandlePawnHealthChanged)) ; Damage will force the pawn out of the couch
      (EventDispatchers|BindEventtoOnDied _pawn (EventDispatchers|CreateEvent :!event Custom|HandleOnDied)) ; When pawn dies on the couch, need to run cleanup.
      (Pawn|Components|CharacterMovement|SetMovementMode (Variables|/Script/Engine.Character|GetCharacterMovement (Variables|CachedPointers|GetPawn)) (Variables|Config|GetSeatedMovementMode) (Variables|Config|GetSeatedCustomModeByte)) ; Custom Mode of 2 is Passenger
      (GameplayCue|ExecuteGameplayCue (Variables|CachedPointers|GetPawn) (Variables|Config|GetSitGameplayCueTag))
      (bind _returnvalue (Utilities|Time|SetTimerbyEvent (EventDispatchers|CreateEvent :!event Custom|OnCanStandTimerEnded) (Variables|Config|GetDelayBeforeCanStand))) ; Temporarily ignore stand-up input, to prevent spamming,\nand perhaps to allow time for the sit-down anim to complete
      (Variables|Dynamic|SetCanStandTimerHandle _returnvalue)
      (GameplayEffects|ApplyGameplayEffectToTarget :self (Variables|CachedPointers|GetASC) :GameplayEffectClass (Variables|Assets|GetGE_VoidSlide) :Level 0.0)
      (GameplayEffects|ApplyGameplayEffectToTarget :self (Variables|CachedPointers|GetASC) :GameplayEffectClass (Variables|Assets|GetGE_HealthRegen) :Level 0.0)
      (GameplayEffects|FortApplyGameplayEffectToTarget :self (Variables|CachedPointers|GetASC) :GameplayEffectClass (Variables|Assets|GetGE_Seated) :TargetOfEffect (Variables|CachedPointers|GetASC) :Level 0.0))
    (:Remote
      ;;[1 1 1] Client-only setup
      (Utilities|FlowControl|Sequence
        (:then_0
          (bind _controller (Variables|CachedPointers|GetController)
            (:then
              (Input|AddMappingContext (PlayerController|LocalPlayerSubsystems|GetEnhancedInputLocalPlayerSubsystem _controller) (Variables|Assets|GetInputMappingContext) (Variables|Config|GetInputPriority) (struct :bIgnoreAllPressedKeysUntilRelease true :bForceImmediately false :bNotifyUserSettings false)) ; Set up input
              (Weapon|SetHolsterWeapon (Variables|CachedPointers|GetPawn) 0 true true (Variables|Const|GetHolsterId)) ; Server holsters, but locally\ncontrolled pawn must also
              (CallFunction|CreateWidget)
              (if (Variables|Config|GetbShouldHideHUD)
                (CallFunction|SetHUDHidden true)))
            (:else)) ; Check if this is a locally controlled player
        )
        (:then_1
          (CallFunction|FixTransform) ; Client pawns sometimes have an incorrect yaw
        )))))

(event EventEndPlay
  (params (EndPlayReason :type EEndPlayReason%20Enum :default "Destroyed"))
  ;[1 0 1 0.5] [EndPlay]
  (Animation|UnlinkAnimLayer (Variables|CachedPointers|GetPawn) "/WildEstate/Gameplay/VoidSlide/Sequence/VoidSlide_Player_AnimBP.VoidSlide_Player_AnimBP_C")
  (Tags|UnBlockAbilitiesWithTags (Variables|CachedPointers|GetASC) (Variables|Default|GetBlockedActions) false)
  (Utilities|SwitchHasAuthority
    (:Authority
      ;;[1 1 1] Server-only cleanup
      (CallFunction|StandUp) ; Failsafe in case we still hadn't stood up
      (if (Variables|Default|GetbRemovedViaInteractionOrDamage)
        (CallFunction|SetPawnExitRotation)
        (Variables|Default|SetbRemovedViaInteractionOrDamage false)))
    (:Remote
      ;;[1 1 1] Client-only cleanup
      (CallFunction|ClientCleanUp))))

(event Custom|Initialize
  (params (SeatIndex :type Integer) (CouchActor :type BGA%20Void%20Slide%20Parent%20Object%20Reference) (CameraTarget :type BP%20Void%20Slide%20Camera%20Target%20Object%20Reference))
  ;[1 1 0 0.5] [Initialize] Called from BGA_VoidSlide_Parent immediately after we get created on server.
  ; Performs setup that we could not handle in BeginPlay, because we didn't yet have all of the necessary data.
  (Variables|Replicated|SetSeatIndex SeatIndex)
  (Utilities|FlowControl|DoOnce false
    (:Completed
      (Utilities|SwitchHasAuthority
        (:Authority
          (Variables|CachedPointers|SetCouchActor CouchActor)
          (Variables|CachedPointers|SetCameraTarget CameraTarget)
          (CallFunction|OverrideViewTarget)
          (GameplayEffects|FortApplyGameplayEffectToTarget :self (Variables|CachedPointers|GetASC) :GameplayEffectClass (Variables|Assets|GetGE_GrantIdle) :TargetOfEffect (Variables|CachedPointers|GetASC) :OptionalSourceObject (Variables|CachedPointers|GetCouchActor) :OptionalEffectCauser (Variables|CachedPointers|GetCouchActor) :Level 0.0)
          (CallFunction|PostSeatedSetup)
          (CallFunction|FixTransform))))))

(event Custom|OverrideViewTarget
  ;[0.00623 0.08701 0.15 0.5] [OverrideViewTarget]
  (Utilities|FlowControl|DoOnce false
    (:Completed
      (Networking|SwitchHasAuthority
        (:Authority
          (if (Variables|Config|GetbUseCustomViewTarget)
            (Variables|CachedPointers|GetCameraTarget
              (:then
                (bind _controller (Variables|CachedPointers|GetController)
                  (:then
                    (Utilities|IsValid (Pawn|GetViewTarget _controller)
                      (:Is%20Valid
                        (Variables|CachedPointers|SetPreviousViewTarget (Pawn|GetViewTarget _controller)) ; Store the ViewTarget, to be restored\nwhen we exit the couch
                        (Game|Player|SetViewTargetwithBlend _controller (Variables|CachedPointers|GetCameraTarget) 0.1 "VTBlend_Linear" 0.0 true) ; Temporarily override the ViewTarget, and hope\nthat nothing else fights for control
                      )))
                  (:else)))
              (:else))))))))

(event Custom|RestoreViewTarget
  ;[0 0.5 0.5 0.5] [RestoreViewTarget]
  (Utilities|FlowControl|DoOnce false
    (:Completed
      (Networking|SwitchHasAuthority
        (:Authority
          (bind _cameratarget (Variables|CachedPointers|GetCameraTarget)
            (:then
              (bind _controller (Variables|CachedPointers|GetController)
                (:then
                  (if (== (Pawn|GetViewTarget _controller) _cameratarget)
                    (Game|Player|SetViewTargetwithBlend _controller (Variables|CachedPointers|GetPreviousViewTarget) 0.01 "VTBlend_Linear" 0.0 true) ; Restore ViewTarget to what it was before we sat down
                  ))
                (:else)))
            (:else)))))))

(event Custom|SetHUDHidden
  (params (bHidden :type Boolean))
  (bind _controller (Variables|CachedPointers|GetController))
  ;[0.02858 0.15 0.0702 0.5] [SetHUDHidden]
  (Networking|CanExecuteCosmeticEvents
    (:True
      (bind _controller_2 (Variables|CachedPointers|GetController)
        (:then
          (HUD|SetHUDElementVisibilitywithContext _controller_2 (Variables|Config|GetHideHUDTags) bHidden "VoidSlide" "")
          (Map|HideFullScreenMap (Class|FortPlayerController|GetMiniMapComponent _controller))
          (if (Class|FortPlayerControllerAthena|IsinInventory _controller)
            (Class|FortPlayerController|ToggleInventory _controller) ; Close the inventory
          ))
        (:else)))))

(event Custom|PostSeatedSetup
  ;[0.5 1 0 0.5] [PostSeatedSetup]
  (Utilities|SwitchHasAuthority
    (:Authority
      (Transformation|AttachActorToComponent (Variables|CachedPointers|GetPawn) (Class|BGAVoidSlideParent|GetCouchMesh (Variables|CachedPointers|GetCouchActor)) (Default|GetSeatSocketName (Variables|/WildEstate/Gameplay/VoidSlide/Blueprints/BGA_VoidSlide_Parent.BGA_VoidSlide_Parent_C|GetSeatComponent (Variables|CachedPointers|GetCouchActor)) (Variables|Replicated|GetSeatIndex)) "SnapToTarget" "SnapToTarget" "KeepRelative" false))))

(event Custom|FixTransform
  (bind (_returnvalue_roll _returnvalue_pitch _returnvalue_yaw) (Math|Rotator|BreakRotator (Transformation|GetActorRotation (Variables|CachedPointers|GetPawn))))
  (bind _seatindex (Variables|Replicated|GetSeatIndex))
  ;[0.15 0.02829 0 0.5] [FixTransform] Sometimes the client transform retains loc or yaw from prior to attachment
  (bind _asbga_void_slide_parent (Utilities|Casting|CastToBGA_VoidSlide_Parent (Actor|GetAttachParentActor (Variables|CachedPointers|GetPawn)))
    (:then
      (if (>= (Variables|Replicated|GetSeatIndex) 0)
        (bind (_returnvalue_location _returnvalue_rotation) (Math|Transform|BreakTransform (Transformation|GetSocketTransform (Class|BGAVoidSlideParent|GetCouchMesh _asbga_void_slide_parent) (Default|GetSeatSocketName (Variables|/WildEstate/Gameplay/VoidSlide/Blueprints/BGA_VoidSlide_Parent.BGA_VoidSlide_Parent_C|GetSeatComponent _asbga_void_slide_parent) _seatindex) "RTS_World")))
        (if (not (and (Math|Float|NearlyEqual(Float) (Math|Vector|DistanceSquared(Vector) _returnvalue_location (Transformation|GetActorLocation (Variables|CachedPointers|GetPawn))) 0.0 1.0) (Math|Float|NearlyEqual(Float) (.yaw _returnvalue_rotation) _returnvalue_yaw 0.1)))
          (Transformation|SetActorTransform (Variables|CachedPointers|GetPawn) (Math|Transform|MakeTransform _returnvalue_location _returnvalue_rotation (Transformation|GetActorScale3D (Variables|CachedPointers|GetPawn))) false true) ; Force the world loc and rot to align with the socket
          (else
            _))
        (Utilities|FlowControl|DelayUntilNextTick) ; Keep trying until the pawn attachment and SeatIndex has been replicated
        (CallFunction|FixTransform)
        (else
          _)))
    (:CastFailed
      _)) ; Wait until attachment has replicated
)

(event Custom|RemoveWidget
  ;[0.15 0.06235 0.1389 0.5] [RemoveWidget] Removes the input/keybinds widget from the HUD
  (bind _widget (Variables|CachedPointers|GetWidget)
    (:then
      (Widget|RemovefromParent _widget))
    (:else)))

(event Custom|CreateWidget
  ;[0 0 1 0.5 17] [CreateWidget] Creates the input/keybinds widget and adds it to AthenaHUDContextualControls' ControllerCalloutbox, which is the same spot that the vehicle controls widget exists, for example.
  (Networking|CanExecuteCosmeticEvents
    (:True
      (Variables|Assets|GetWidgetClass
        (:then
          (bind _foundwidgets (Widget|GetAllWidgetsOfClass (Variables|Assets|GetParentWidgetClass) false)) ; Get a ref to the HUD widget we will be attaching our input widget to
          (if (Utilities|Array|IsNotEmpty _foundwidgets)
            (bind _asathena_hudcontextual_controls (Utilities|Casting|CastToAthenaHUDContextualControls (Utilities|Array|Get(acopy) _foundwidgets))
              (:then
                (bind _returnvalue (UserInterface|CreateWidget (Variables|Assets|GetWidgetClass) (Variables|CachedPointers|GetController))) ; Create the input widget
                (bind _output_get (Variables|CachedPointers|SetWidget _returnvalue)) ; Keep a ref for later removal
                (Widget|Panel|AddChild (Variables|/Game/Athena/HUD/AthenaHUDContextualControls.AthenaHUDContextualControls_C|GetControllerCalloutBox _asathena_hudcontextual_controls) _output_get) ; Add it to the ControllerCalloutBox\nof AthenaHUDContextualControls
              )
              (:CastFailed
                (Fortnite|FortShippingLog self "CreateWidget: Failed cast to AthenaHUDContextualControls" true)))
            (else
              (Fortnite|FortShippingLog self "CreateWidget: GetAllWidgetsOfClass returned empty" true))))
        (:else)))))

(event EnhancedInputActionVoidSlideChangeSeatInput
  (params (ActionValue :type Boolean) (ElapsedSeconds :type Float%20%28double-precision%29) (TriggeredSeconds :type Float%20%28double-precision%29) (InputAction :type Input%20Action%20Object%20Reference :default "/WildEstate/Gameplay/VoidSlide/Blueprints/VoidSlideChangeSeatInput.VoidSlideChangeSeatInput")))

(event Custom|Server_ChangeSeat
  ;[1 1 0 0.5] [Server_ChangeSeat]
  (if (not (Utilities|Time|IsTimerActivebyHandle (Variables|Dynamic|GetCanChangeSeatTimerHandle)))
    (bind (_bchangedseat _seatindex) (Class|BPVoidSlideSeatComponent|TryChangeSeat (Variables|/WildEstate/Gameplay/VoidSlide/Blueprints/BGA_VoidSlide_Parent.BGA_VoidSlide_Parent_C|GetSeatComponent (Variables|CachedPointers|GetCouchActor)) (Variables|CachedPointers|GetPlayerState)))
    (if _bchangedseat
      (Variables|Replicated|SetSeatIndex _seatindex) ; Replicate the change
      (CallFunction|PostSeatedSetup) ; Handle attachment, anim, etc.
      (bind _returnvalue (Utilities|Time|SetTimerbyEvent (EventDispatchers|CreateEvent :!event Custom|OnCanChangeSeatTimerEnded) (Variables|Config|GetDelayBeforeCanChangeSeat))) ; Temporarily ignore change-seat input, to prevent spamming,\nand perhaps to allow time for the anim to complete
      (Variables|Dynamic|SetCanChangeSeatTimerHandle _returnvalue))))

(event Custom|OnCanChangeSeatTimerEnded)

(event EnhancedInputActionVoidSlideExitInput
  (params (ActionValue :type Boolean) (ElapsedSeconds :type Float%20%28double-precision%29) (TriggeredSeconds :type Float%20%28double-precision%29) (InputAction :type Input%20Action%20Object%20Reference :default "/WildEstate/Gameplay/VoidSlide/Blueprints/VoidSlideExitInput.VoidSlideExitInput")))

(event Custom|Server_Exit
  ;[0.15 0.05951 0.03824 0.5] [Server_Exit]
  (if (Utilities|Time|IsTimerActivebyHandle (Variables|Dynamic|GetCanStandTimerHandle))
    _
    (elif (Utilities|Time|IsTimerActivebyHandle (Variables|Dynamic|GetCanChangeSeatTimerHandle))
      _
      (else
        (Utilities|FlowControl|DoOnce false
          (:Completed
            (Variables|Default|SetbRemovedViaInteractionOrDamage true)
            (CallFunction|StandUp)))))))

(event Custom|HandlePawnHealthChanged
  (bind _pawn (Variables|CachedPointers|GetPawn))
  ;[1 1 0 0.5] [HandlePawnHealthChanged] Force the player out of the seat, if health or shield reduced
  (Utilities|SwitchHasAuthority
    (:Authority
      (if (or (< (FortPawn|GetHealth (Variables|CachedPointers|GetPawn)) (Variables|Dynamic|GetPreviousHealth)) (< (FortPawn|GetShield (Variables|CachedPointers|GetPawn)) (Variables|Dynamic|GetPreviousShield)))
        (Utilities|FlowControl|DoOnce false
          (:Completed
            (Variables|Default|SetbRemovedViaInteractionOrDamage true)
            (CallFunction|StandUp) ; Force the player out of the seat
          ))
        (else
          (Variables|Dynamic|SetPreviousHealth (FortPawn|GetHealth _pawn))
          (Variables|Dynamic|SetPreviousShield (FortPawn|GetShield _pawn)))))))

(event Custom|StandUp
  (bind _pawn (Variables|CachedPointers|GetPawn))
  ;[0.04205 0.15 0.04308 0.5] [StandUp]
  (Utilities|FlowControl|DoOnce false
    (:Completed
      (Networking|SwitchHasAuthority
        (:Authority
          (Utilities|Time|ClearandInvalidateTimerbyHandle (Variables|Dynamic|GetCanStandTimerHandle))
          (Utilities|Time|ClearandInvalidateTimerbyHandle (Variables|Dynamic|GetCanChangeSeatTimerHandle))
          (Health|UnbindEventfromOnPawnHealthChanged _pawn (EventDispatchers|CreateEvent :!event Custom|HandlePawnHealthChanged))
          (EventDispatchers|UnbindEventfromOnDied _pawn (EventDispatchers|CreateEvent :!event Custom|HandleOnDied))
          (Default|Unoccupy (Variables|/WildEstate/Gameplay/VoidSlide/Blueprints/BGA_VoidSlide_Parent.BGA_VoidSlide_Parent_C|GetSeatComponent (Variables|CachedPointers|GetCouchActor)) (Variables|CachedPointers|GetPlayerState))
          (Transformation|DetachFromActor (Variables|CachedPointers|GetPawn) "KeepWorld" "KeepWorld" "KeepWorld")
          (CallFunction|RestoreViewTarget)
          (Weapon|SetHolsterWeapon (Variables|CachedPointers|GetPawn) 0 false true (Variables|Const|GetHolsterId))
          (GameplayEffects|RemoveActiveGameplayEffectbySourceEffect (Variables|CachedPointers|GetASC) (Variables|Assets|GetGE_VoidSlide) 0 1)
          (GameplayEffects|RemoveActiveGameplayEffectbySourceEffect (Variables|CachedPointers|GetASC) (Variables|Assets|GetGE_HealthRegen) 0 1)
          (GameplayEffects|RemoveActiveGameplayEffectbySourceEffect (Variables|CachedPointers|GetASC) (Variables|Assets|GetGE_GrantIdle) 0 1)
          (GameplayEffects|RemoveActiveGameplayEffectbySourceEffect (Variables|CachedPointers|GetASC) (Variables|Assets|GetGE_Seated) 0 1)
          (CallFunction|EjectandLaunchPlayer)
          (CallFunction|ExecuteClientCleanup))))))

(event Custom|EjectAndLaunchPlayer
  (bind _pawn (Variables|CachedPointers|GetPawn))
  (bind _charactermovement (Variables|/Script/Engine.Character|GetCharacterMovement _pawn))
  ;[1 1 1] [LaunchPlayer] - Launch the player slightly when they stand up from the couch.
  (Components|Movement|StopMovementImmediately _charactermovement)
  (Pawn|Components|CharacterMovement|SetMovementMode _charactermovement "MOVE_Walking" 0)
  (AthenaMovement|SetPendingSkydiveLaunch _pawn false)
  (Collision|SetActorEnableCollision _pawn true)
  ;;[0.5052 0.5052 0.5052 0.5] Raise the player slightly and temporarily ignore Pawn collisions in order to eject from couch cleanly
  (Transformation|SetActorLocation _pawn (+ (Transformation|GetActorLocation _pawn) (Variables|Config|GetPawnEjectOffsetPosition)) false false)
  (Collision|SetCollisionResponsetoChannel (Variables|/Script/Engine.Character|GetCapsuleComponent _pawn) "ECC_Pawn" "ECR_Ignore")
  (Utilities|Time|SetTimerbyEvent (EventDispatchers|CreateEvent :!event Custom|CleanupAfterLaunch) (Variables|Config|GetReenablePawnCollisionTimeSecond))
  (Character|LaunchCharacter (Variables|CachedPointers|GetPawn) (Utilities|Operators|Add (* (Transformation|GetActorForwardVector (Variables|CachedPointers|GetPawn)) (Variables|Config|GetLaunchForwardSpeed)) (Math|Vector|MakeVector 0.0 0.0 (Variables|Config|GetLaunchUpSpeed))) true true))

(event Custom|CleanupAfterLaunch
  (bind _pawn (Variables|CachedPointers|GetPawn))
  ;[0 0 0 0.5] [EnableCollisionAfterLaunch] - Re-enable Pawn collision and destroy the component.
  (Collision|SetActorEnableCollision _pawn true)
  (Collision|SetCollisionResponsetoChannel (Variables|/Script/Engine.Character|GetCapsuleComponent _pawn) "ECC_Pawn" "ECR_Block")
  (CallFunction|SetPawnExitRotation)
  (Components|DestroyComponent))

(event Custom|HandleOnDied
  (params (DamagedActor :type Actor%20Object%20Reference) (Damage :type Float%20%28single-precision%29) (InstigatedBy :type Controller%20Object%20Reference) (DamageCauser :type Actor%20Object%20Reference) (HitLocation :type Vector) (FHitComponent :type Primitive%20Component%20Object%20Reference) (BoneName :type Name) (Momentum :type Vector))
  ; HandleOnDied - When Pawn dies, run the client cleanup.
  (CallFunction|ExecuteClientCleanup))

(event Custom|ExecuteClientCleanup
  ; HandleOnDied - When Pawn dies, run the client cleanup.
  (CallFunction|ClientCleanUp))

(fn BlockAbilities
  (bind (_voidslidetablerow_bshouldapplytags_1_f01822b84acc14c79e076e935869932f _voidslidetablerow_tagstoapply_4_7cfb82444884dbf0eb0a6ebf3b2c4244) (Utilities|Struct|BreakVoidSlideAbilityTagStruct (Variables|GetVoidSlideTableRow_bShouldApplyTags_1_F01822B84ACC14C79E076E935869932F)))
  (bind (_voidslidetablerow_bshouldapplytags_1_f01822b84acc14c79e076e935869932f_2 _voidslidetablerow_tagstoapply_4_7cfb82444884dbf0eb0a6ebf3b2c4244_2) (Utilities|Struct|BreakVoidSlideAbilityTagStruct (Variables|GetVoidSlideTableRow_bShouldApplyTags_1_F01822B84ACC14C79E076E935869932F)))
  (Utilities|FlowControl|Sequence
    (:then_0
      (Networking|SwitchHasAuthority
        (:Authority
          (bind _returnvalue (Utilities|GetDataTableRow "/WildEstate/Gameplay/VoidSlide/DataTables/VoidSlideTagsTable.VoidSlideTagsTable" "VoidSlide.Server.BlockAbilityTags")
            (:then
              (LocalVariables|SetVoidSlideTableRow _returnvalue))
            (:RowNotFound
              _)))
        (:Remote
          (bind _returnvalue_2 (Utilities|GetDataTableRow "/WildEstate/Gameplay/VoidSlide/DataTables/VoidSlideTagsTable.VoidSlideTagsTable" "VoidSlide.Client.BlockAbilityTags")
            (:then
              (LocalVariables|SetVoidSlideTableRow _returnvalue_2))
            (:RowNotFound
              _))))
      (Fortnite|FortShippingLog self "[BlockAbilities] Failed to find row in VoidSlideTagsTable" true))
    (:then_1
      (Fortnite|FortShippingLog self (Utilities|String|Append "[BlockAbilities] bShouldApplyTags [" (Utilities|String|ToString(Boolean) _voidslidetablerow_bshouldapplytags_1_f01822b84acc14c79e076e935869932f) "] TagsToApply.Num [" (Utilities|String|ToString(Integer) (GameplayTags|GetNumGameplayTagsinContainer _voidslidetablerow_tagstoapply_4_7cfb82444884dbf0eb0a6ebf3b2c4244)) "]")))
    (:then_2
      (if _voidslidetablerow_bshouldapplytags_1_f01822b84acc14c79e076e935869932f_2
        (Fortnite|FortShippingLog self "[BlockAbilities] Calling ASC->BlockAbilitiesWithTags")
        (bind _output_get (Variables|Default|SetBlockedActions _voidslidetablerow_tagstoapply_4_7cfb82444884dbf0eb0a6ebf3b2c4244_2)) ; Track the list of blocked actions,\nso we can unblock them in EndPlay
        (Tags|BlockAbilitiesWithTags (Variables|CachedPointers|GetASC) _output_get)))))

(fn CancelAbilities
  (bind (_voidslidetablerow_bshouldapplytags_1_f01822b84acc14c79e076e935869932f _voidslidetablerow_tagstoapply_4_7cfb82444884dbf0eb0a6ebf3b2c4244) (Utilities|Struct|BreakVoidSlideAbilityTagStruct (Variables|GetVoidSlideTableRow_bShouldApplyTags_1_F01822B84ACC14C79E076E935869932F)))
  (bind (_voidslidetablerow_bshouldapplytags_1_f01822b84acc14c79e076e935869932f_2 _voidslidetablerow_tagstoapply_4_7cfb82444884dbf0eb0a6ebf3b2c4244_2) (Utilities|Struct|BreakVoidSlideAbilityTagStruct (Variables|GetVoidSlideTableRow_bShouldApplyTags_1_F01822B84ACC14C79E076E935869932F)))
  (Utilities|FlowControl|Sequence
    (:then_0
      (Networking|SwitchHasAuthority
        (:Authority
          (bind _returnvalue (Utilities|GetDataTableRow "/WildEstate/Gameplay/VoidSlide/DataTables/VoidSlideTagsTable.VoidSlideTagsTable" "VoidSlide.Server.CancelAbilityTags")
            (:then
              (LocalVariables|SetVoidSlideTableRow _returnvalue))
            (:RowNotFound
              _)))
        (:Remote
          (bind _returnvalue_2 (Utilities|GetDataTableRow "/WildEstate/Gameplay/VoidSlide/DataTables/VoidSlideTagsTable.VoidSlideTagsTable" "VoidSlide.Client.CancelAbilityTags")
            (:then
              (LocalVariables|SetVoidSlideTableRow _returnvalue_2))
            (:RowNotFound
              _))))
      (Fortnite|FortShippingLog self "[CancelAbilities] Failed to find row in VoidSlideTagsTable" true))
    (:then_1
      (Fortnite|FortShippingLog self (Utilities|String|Append "[CancelAbilities] bShouldApplyTags [" (Utilities|String|ToString(Boolean) _voidslidetablerow_bshouldapplytags_1_f01822b84acc14c79e076e935869932f) "] TagsToApply.Num [" (Utilities|String|ToString(Integer) (GameplayTags|GetNumGameplayTagsinContainer _voidslidetablerow_tagstoapply_4_7cfb82444884dbf0eb0a6ebf3b2c4244)) "]")))
    (:then_2
      (if _voidslidetablerow_bshouldapplytags_1_f01822b84acc14c79e076e935869932f_2
        (Fortnite|FortShippingLog self "[CancelAbilities] Calling ASC->CancelAbilitiesWithTags")
        (Tags|CancelAbilitiesWithTags (Variables|CachedPointers|GetASC) _voidslidetablerow_tagstoapply_4_7cfb82444884dbf0eb0a6ebf3b2c4244_2)))))

(fn ClientCleanUp
  (bind _controller (Variables|CachedPointers|GetController)
    (:then
      (Input|RemoveMappingContext (PlayerController|LocalPlayerSubsystems|GetEnhancedInputLocalPlayerSubsystem _controller) (Variables|Assets|GetInputMappingContext) (struct :bIgnoreAllPressedKeysUntilRelease true :bForceImmediately false :bNotifyUserSettings false))
      (Weapon|SetHolsterWeapon (Variables|CachedPointers|GetPawn) 0 false true (Variables|Const|GetHolsterId)) ; Server unholsters, but locally\ncontrolled pawn must also
      (CallFunction|RemoveWidget)
      (if (Variables|Default|GetbRemovedViaInteractionOrDamage)
        (CallFunction|SetPawnExitRotation)
        (Variables|Default|SetbRemovedViaInteractionOrDamage false)
        (else
          _))
      (if (Variables|Config|GetbShouldHideHUD)
        (CallFunction|SetHUDHidden false)))
    (:else)))

(fn GetAnimData
  (return (Variables|Replicated|GetSeatIndex)))

(fn SetPawnExitRotation
  (bind _couchactor (Variables|CachedPointers|GetCouchActor))
  (bind _returnvalue (Transformation|GetActorRightVector _couchactor))
  (bind _returnvalue_2 (Math|Rotator|MakeRotfromX _returnvalue))
  (if (and (Utilities|IsValid (Variables|CachedPointers|GetController) :!class /Script/Engine.KismetSystemLibrary) (Utilities|IsValid (Variables|CachedPointers|GetCouchActor) :!class /Script/Engine.KismetSystemLibrary))
    (Fortnite|KeepCGPathOptimizations (Variables|CachedPointers|GetController)) ; Mark the frame for a safe camera write in order to avoid anticheat.  \nAnticheat doesn't like us setting the control rotation here
    (Transformation|SetActorRotation (Variables|CachedPointers|GetPawn) _returnvalue_2 false)
    (Pawn|SetControlRotation (Variables|CachedPointers|GetController) _returnvalue_2)
    (else
      (Fortnite|FortShippingLog self "SetPawnExitRotation: Could not set pawn exit rotation due to controller or couch actor being null." true))))
