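# Niagara system for the VoidSlide heal burst, a brief green flash attached to a user-supplied skeletal mesh.
# Decal_New bursts 17 sprites, each following the bone or socket matching its particle ID with an array-picked offset and curve-animated material parameter.
# Flares layers corona, anamorphic and distant-glow sprites at index 16, beside Empty's fading light.
# These bursts use System.IsFirst as spawn probability; the two remaining emitters list no renderers.
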NiagaraSystem /WildEstate/Gameplay/VoidSlide/FX/NS_VoidSlide_Healburst
dependencies {
 usedDataInterfaces [&NiagaraDataInterfaceCurve, &NiagaraDataInterfaceArrayFloat3]
 usedDynamicInputs [&MakeCustomFloatFromBool]
 usedModules [
  &self:SystemSpawnScript.SystemScriptSource.SystemScriptGraph.NiagaraNodeAssignment_0.SetVariables_#1
  &self:SystemSpawnScript.SystemScriptSource.SystemScriptGraph.NiagaraNodeAssignment_1.SetVariables_#2
  &SystemState
 ]
}
emitters {
 +s {rendererClasses [], renderers []}
 Decal_New {
  emitterUpdateScript [
   &EmitterState()
   &SpawnBurst_Instantaneous(Spawn Count=17, Spawn Probability=$System.IsFirst, Use Spawn Probability=true, UseLoopCountLimit=false)
  ]
  particleSpawnScript [
   &InitializeParticle(
    Color=<rgba 0.0213 1 0.3426 0.1>
    Sprite Size Mode=::NewEnumerator3
    Uniform Sprite Size=40
    UsePositionOffset=false
    Write Material Random=true
   )
   &CameraOffset(Camera Offset Amount=25)
  ]
  particleUpdateScript [
   &ParticleState()
   &SkeletalMeshLocation(
    Bone / Socket Index=&ReturnParticleIDAttributes()
    Bone Sampling Mode=::NewEnumerator4
    MaskSpawnGroup=false
    Sampled Position Offset=&SelectVectorFromArray(
     Direct Array Index=&ReturnParticleIDAttributes()
     Vector Selection Array=DataInterface() {propertyValues {FloatData [
        <0 0 0>
        <5 0 0>
        <0 0 0>
        <0 0 0>
        <5 0 0>
        <0 0 0>
        <0 0 0>
        <-5 0 0>
        <-10 0 0>
        <8 0 0>
        <-8 0 0>
        <-4 0 0>
        <8 10 0>
        <8 0 0>
        <4 0 0>
        <-8 -10 0>
        <-10 -22 0>
       ]}}
    )
    Skeletal Mesh=$User.SkeletalMesh
   )
   &DynamicMaterialParameters(
    Index 0 Param 1=&FloatFromCurve(
     FloatCurve=DataInterface() {propertyValues {Curve {
        defaultValue=1
        keys [
         {arriveTangent=3.695, interpMode=RCIM_Cubic, time=0, value=-0.5}
         {arriveTangentWeight=0.1795, interpMode=RCIM_Cubic, leaveTangentWeight=0.718, tangentMode=RCTM_User, time=0.2, value=0}
         {arriveTangent=-0.0982, leaveTangent=2.7414, tangentMode=RCTM_User, time=1, value=2}
        ]
       }}}
     Scale Curve=4
    )
    Module.Index0Float0Write=true
    Module.Index0Float1Write=true
    Module.Index0Float2Write=true
    Module.Index0Float3Write=true
   )
   &DynamicMaterialParameters(
    Index 0 Param 1=&FloatFromCurve(FloatCurve=DataInterface() {propertyValues {Curve {defaultValue=1, keys [<at 0 0 RCIM_Cubic>, {arriveTangent=2.2458, tangentMode=RCTM_User, time=1, value=1}]}}})
    Module.Index0Float0Write=true
    Module.Index0Float1Write=true
    Module.Index0Float2Write=true
    Module.Index0Float3Write=true
   ) {enabled=false, module=DynamicMaterialParameters001}
   &ScaleColor(
    Scale Alpha=&FloatFromCurve(
     FloatCurve=DataInterface() {propertyValues {Curve {
        defaultValue=1
        keys [
         {arriveTangent=5.904, interpMode=RCIM_Cubic, leaveTangent=5.904, tangentMode=RCTM_User, time=0, value=0}
         {arriveTangentWeight=0.0667, interpMode=RCIM_Cubic, tangentMode=RCTM_User, time=0.2, value=1}
         {arriveTangent=-2.3227, arriveTangentWeight=0.2667, interpMode=RCIM_Cubic, leaveTangent=-2.3227, tangentMode=RCTM_User, time=1, value=0}
        ]
       }}}
     Scale Curve=0.05
    )
    Scale RGB=&VectorFromFloat(Value=0.15)
    ScaleA=true
    ScaleRGB=true
   )
  ]
  rendererClasses [&NiagaraSpriteRendererProperties]
  renderers [{rendererClass=&NiagaraSpriteRendererProperties}]
 }
 Empty {
  emitterUpdateScript [
   &EmitterState()
   &SpawnBurst_Instantaneous(Spawn Probability=$System.IsFirst, Use Spawn Probability=true, UseLoopCountLimit=false)
  ]
  particleSpawnScript [
   &InitializeParticle(Color=<rgba 0.0213 1 0.3426 1>, Lifetime=0.2, UsePositionOffset=true, Write Material Random=true)
   &SkeletalMeshLocation(
    Bone / Socket Index=16
    Bone Sampling Mode=::NewEnumerator4
    MaskSpawnGroup=false
    Sampled Position Offset=<0 50 0>
    Skeletal Mesh=$User.SkeletalMesh
   )
  ]
  particleUpdateScript [
   &ParticleState()
   /SetVariables(Particles.LightExponent=3, Particles.LightRadius=150) {isSetParameters=true}
   &ScaleColor(
    Scale RGB=&VectorFromFloat(Value=&FloatFromCurve(
      FloatCurve=DataInterface() {propertyValues {Curve {keys [
          {interpMode=RCIM_Cubic, leaveTangentWeight=0.4714, tangentMode=RCTM_User, time=0, value=1}
          {arriveTangentWeight=0.4714, interpMode=RCIM_Cubic, tangentMode=RCTM_User, time=1, value=0}
         ]}}}
      Scale Curve=5
     ))
    ScaleA=true
    ScaleRGB=true
   )
  ]
  rendererClasses [&NiagaraLightRendererProperties]
  renderers [{rendererClass=&NiagaraLightRendererProperties, rendererIndex=0}]
 }
 Flares {
  emitterUpdateScript [
   &EmitterState()
   &SpawnBurst_Instantaneous(Spawn Probability=$System.IsFirst, Use Spawn Probability=true, UseLoopCountLimit=false)
  ]
  particleSpawnScript [
   &InitializeParticle(
    Color=<rgba 0 0.5 0.1642 2>
    Lifetime=0.2
    Sprite Rotation Angle=0.2
    Sprite Rotation Mode=::NewEnumerator1
    Sprite Size Mode=::NewEnumerator3
    Uniform Sprite Size=75
    UsePositionOffset=false
    Write Material Random=true
   )
   &CameraOffset(Camera Offset Amount=75)
   /SetVariables(Particles.AnamorphicSize:Vector2f=&Multiply_Vector2D_ByFloat(Float=0.5, Vector2D=<240 7>)) {isSetParameters=true}
   &SkeletalMeshLocation(Bone / Socket Index=16, Bone Sampling Mode=::NewEnumerator4, MaskSpawnGroup=false, Skeletal Mesh=$User.SkeletalMesh)
  ]
  particleUpdateScript [
   &ParticleState()
   &ScaleColor(
    Scale Alpha=&FloatFromCurve(FloatCurve=DataInterface() {propertyValues {Curve {keys [
         {arriveTangent=0.6364, interpMode=RCIM_Cubic, leaveTangent=0.6364, time=0, value=0}
         {arriveTangent=1.4286, interpMode=RCIM_Cubic, leaveTangent=1.4286, time=0.55, value=0.35}
         {arriveTangent=-8.2069, interpMode=RCIM_Cubic, leaveTangent=-8.2069, tangentMode=RCTM_User, time=0.7, value=1}
         {arriveTangent=-0.0405, interpMode=RCIM_Cubic, leaveTangent=-0.0405, tangentMode=RCTM_User, time=1, value=0}
        ]}}})
    ScaleA=true
    ScaleRGB=true
   )
   &ScaleSpriteSize(Uniform Curve Sprite Scale=DataInterface() {propertyValues {Curve {keys [
        {interpMode=RCIM_Cubic, leaveTangentWeight=0.1863, tangentMode=RCTM_User, time=0, value=0.4}
        {
         arriveTangent=0.3825
         arriveTangentWeight=0.2415
         interpMode=RCIM_Cubic
         leaveTangent=0.3825
         leaveTangentWeight=0.0659
         tangentMode=RCTM_User
         time=0.55
         value=0.5
        }
        {
         arriveTangent=-1.9384
         arriveTangentWeight=0.0547
         interpMode=RCIM_Cubic
         leaveTangent=-1.9384
         leaveTangentWeight=0.1094
         tangentMode=RCTM_User
         time=0.7
         value=1
        }
        {arriveTangentWeight=0.1414, interpMode=RCIM_Cubic, tangentMode=RCTM_User, time=1, value=0.7}
       ]}}})
   /SetVariables(
    Particles.Anamorphic Color:LinearColor=&MakeLinearColorFromVectorAndFloat(
     "Vector (RGB)"=&MakeVectorFromLinearColorRGB(Color=<rgba 0 1.2 0.394 1>)
     "Float (Alpha)"=&Multiply_Float(
      A=0.2
      B=&FloatFromCurve(FloatCurve=DataInterface() {propertyValues {Curve {keys [
           {arriveTangent=0.6364, interpMode=RCIM_Cubic, leaveTangent=0.6364, time=0, value=0}
           {arriveTangent=1.4286, interpMode=RCIM_Cubic, leaveTangent=1.4286, time=0.55, value=0.35}
           {arriveTangent=-8.2069, interpMode=RCIM_Cubic, leaveTangent=-8.2069, tangentMode=RCTM_User, time=0.7, value=1}
           {arriveTangent=-0.0405, interpMode=RCIM_Cubic, leaveTangent=-0.0405, tangentMode=RCTM_User, time=1, value=0}
          ]}}})
     )
    )
    Particles.CoronaColor:LinearColor=&MakeLinearColorFromVectorAndFloat(
     "Vector (RGB)"=&MakeVectorFromLinearColorRGB(Color=<rgba 0 1.2 0.394 1>)
     "Float (Alpha)"=&FloatFromCurve(FloatCurve=DataInterface() {propertyValues {Curve {keys [
          {interpMode=RCIM_Cubic, leaveTangentWeight=0.3578, tangentMode=RCTM_User, time=0, value=0}
          {arriveTangent=0.6909, interpMode=RCIM_Cubic, leaveTangent=0.6909, tangentMode=RCTM_User, time=0.5, value=0.95}
          {arriveTangentWeight=0.1431, interpMode=RCIM_Cubic, leaveTangentWeight=0.2147, tangentMode=RCTM_User, time=0.7, value=1}
          {arriveTangentWeight=0.348, interpMode=RCIM_Cubic, tangentMode=RCTM_User, time=1, value=0}
         ]}}})
    )
    Particles.CoronaDynParameters:Vector4f=&MakeVector4(
     X=&FloatFromCurve(FloatCurve=DataInterface() {propertyValues {Curve {keys [
          {interpMode=RCIM_Cubic, leaveTangentWeight=0.1904, tangentMode=RCTM_User, time=0, value=1.4}
          {
           arriveTangent=-1.2888
           arriveTangentWeight=0.2753
           interpMode=RCIM_Cubic
           leaveTangent=-1.2888
           leaveTangentWeight=0.0918
           tangentMode=RCTM_User
           time=0.525
           value=1.175
          }
          {arriveTangent=-1.9474, interpMode=RCIM_Cubic, leaveTangent=-1.9474, time=0.7, value=0.55}
          {arriveTangentWeight=0.1414, interpMode=RCIM_Cubic, tangentMode=RCTM_User, time=1, value=0.25}
         ]}}})
     Y=8
     Z=0.6
     W=1
    )
    Particles.CoronaSize:Vector2f=&Vector2DFromFloat(Value=&Multiply_Float(
      A=37.5
      B=&FloatFromCurve(FloatCurve=DataInterface() {propertyValues {Curve {keys [
           {interpMode=RCIM_Cubic, leaveTangentWeight=0.3578, tangentMode=RCTM_User, time=0, value=0}
           {
            arriveTangent=1.1981
            arriveTangentWeight=0.2906
            interpMode=RCIM_Cubic
            leaveTangent=1.1981
            leaveTangentWeight=0.1163
            tangentMode=RCTM_User
            time=0.5
            value=0.95
           }
           {arriveTangentWeight=0.1431, interpMode=RCIM_Cubic, leaveTangentWeight=0.2147, tangentMode=RCTM_User, time=0.7, value=1}
          ]}}})
     ))
    Particles.DistanceGlowSize:Vector2f=&Vector2DFromFloat(Value=&Multiply_Float(
      A=350
      B=&FloatFromCurve(FloatCurve=DataInterface() {propertyValues {Curve {keys [
           {interpMode=RCIM_Cubic, leaveTangentWeight=0.348, tangentMode=RCTM_User, time=0, value=0.7}
           {arriveTangent=0.7107, arriveTangentWeight=0.348, interpMode=RCIM_Cubic, leaveTangent=0.7107, tangentMode=RCTM_User, time=1, value=1}
          ]}}})
     ))
    Particles.DistantGlowColor:LinearColor=&MakeLinearColorFromVectorAndFloat(
     "Vector (RGB)"=&MakeVectorFromLinearColorRGB(Color=<rgba 0.0213 1 0.3426 1>)
     "Float (Alpha)"=&FloatFromCurve(FloatCurve=DataInterface() {propertyValues {Curve {keys [
          {arriveTangent=9.0725, interpMode=RCIM_Cubic, leaveTangent=9.0725, tangentMode=RCTM_User, time=0, value=0}
          {arriveTangentWeight=0.0905, interpMode=RCIM_Cubic, leaveTangentWeight=0.8147, tangentMode=RCTM_User, time=0.1, value=0.3}
          {arriveTangent=-0.0405, interpMode=RCIM_Cubic, leaveTangent=-0.0405, tangentMode=RCTM_User, time=1, value=0}
         ]}}})
    )
   ) {isSetParameters=true}
  ]
  rendererClasses [&NiagaraSpriteRendererProperties]
  renderers [
   {rendererClass=&NiagaraSpriteRendererProperties}
   {rendererClass=&NiagaraSpriteRendererProperties, rendererIndex=1}
   {rendererClass=&NiagaraSpriteRendererProperties, rendererIndex=2}
   {rendererClass=&NiagaraSpriteRendererProperties, rendererIndex=3}
  ]
 }
 SparkleStationary {rendererClasses [], renderers []}
}
summary {
 systemName=NS_VoidSlide_Healburst
 emitters [
  {emitterName=Flares, rendererClasses [&NiagaraSpriteRendererProperties]}
  {emitterName=Empty, rendererClasses [&NiagaraLightRendererProperties]}
  {emitterName=Decal_New, rendererClasses [&NiagaraSpriteRendererProperties]}
  {emitterName=SparkleStationary, rendererClasses []}
  {emitterName=+s, rendererClasses []}
 ]
 userVariables [
  User.IsFirstCount {type=NiagaraBool}
  User.SkeletalMesh {
   type=NiagaraDataInterfaceSkeletalMesh
   defaultValue {
    struct=NiagaraExt_VariableValue_DataInterface
    value {dataInterface=&self:NiagaraDataInterfaceSkeletalMesh_0}
   }
  }
 ]
}
