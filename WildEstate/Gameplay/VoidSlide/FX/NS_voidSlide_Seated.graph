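# Niagara one-shot sprite effect for the VoidSlide gameplay feature, likely a small grey puff played when something is seated.
# A single emitter bursts a batch of short-lived, semi-transparent sprites shortly after start, launched in a wide hollow cone around the vertical axis at random speeds.
# Drag and weakened gravity slow them while they grow and fade out over life.
# Dynamic material parameters feed the sprite material a grey colour and a random value.
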NiagaraSystem /WildEstate/Gameplay/VoidSlide/FX/NS_voidSlide_Seated
dependencies {usedDataInterfaces [&NiagaraDataInterfaceCurve], usedDynamicInputs [], usedModules [&SystemState]}
emitters {SimpleSpriteBurst {
  emitterUpdateScript [
   &EmitterState()
   &SpawnBurst_Instantaneous(Spawn Count=16, Spawn Time=0.1, Use Spawn Probability=false, UseLoopCountLimit=false)
  ]
  particleSpawnScript [
   &InitializeParticle(
    Color=<rgba 0.35 0.35 0.35 0.5>
    Lifetime Max=0.2
    Lifetime Min=0.35
    Lifetime Mode=::NewEnumerator1
    Sprite Rotation Mode=::NewEnumerator2
    Sprite Size Mode=::NewEnumerator3
    Uniform Sprite Size=50
    UsePositionOffset=true
    Write Material Random=true
   )
   &AddVelocity(
    Cone Angle=180
    Cone Axis=<0 0 1>
    Inner Cone Angle=120
    Module.UseConeFalloff=false
    Module.UseVelDistribution=false
    Override Seed=false
    Rotation Coordinate Space=::World
    Velocity Mode=::NewEnumerator2
    Velocity Speed=&RandomRangeFloat(Minimum=175, Maximum=22.5)
   )
  ]
  particleUpdateScript [
   &ParticleState()
   &ScaleSpriteSize(Uniform Curve Sprite Scale=DataInterface() {propertyValues {Curve {keys [{tangentMode=RCTM_User, time=0, value=0.5}, {tangentMode=RCTM_User, time=1, value=1}]}}})
   &ScaleColor(
    Scale Alpha=&FloatFromCurve(FloatCurve=DataInterface() {propertyValues {Curve {keys [<at -0.0046 1.0113 RCIM_Cubic>,<at 1.0011 0.002 RCIM_Cubic>]}}})
    ScaleA=true
    ScaleRGB=false
   )
   &Drag(Drag=3, Use Linear Drag=true, Use Rotational Drag=true)
   &AccelerationForce(Acceleration=&Multiply_Vector(A=<0 0 -980>, B=&VectorFromFloat(Value=0.1)))
   &SolveForcesAndVelocity(Clamp Velocity=false, Limit Acceleration=false)
   &DynamicMaterialParameters(
    Index 0 Param 4=&RandomRangeFloat(Minimum=0, Maximum=100)
    Index 0 Vector 1:Vector3f=&MakeVectorFromLinearColorRGB(Color=<rgba 0.15 0.15 0.15 1>)
    Module.Index0Float3Write=true
    Module.Index0VectorWrite=true
    Write Parameter Index 0=::NewEnumerator3
   )
  ]
  rendererClasses [&NiagaraSpriteRendererProperties]
  renderers [{rendererClass=&NiagaraSpriteRendererProperties}]
 }}
summary {
 systemName=NS_voidSlide_Seated
 emitters [{emitterName=SimpleSpriteBurst, rendererClasses [&NiagaraSpriteRendererProperties]}]
 userVariables []
}
